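# Masked, two-sided Default Lit material for the broken window of the StainlessBusted skin on the Future vehicle body.
# It holds two texture samples, a broken-window RGB texture and a matching normal map, with no connections shown between them and the material outputs.
# It is flagged for skeletal meshes and mesh deformers, with automatic usage flagging and ray-traced shadow casting turned off.
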Material /VehicleCosmetics/Bodies/Future/Skins/StainlessBusted/MAT_Skin_StainlessBusted_Window
graph {
 properties {
  bAutomaticallySetUsageInEditor=false
  bCastRayTracedShadows=false
  bUsedWithMeshDeformer=true
  bUsedWithSkeletalMesh=true
  blendMode=BLEND_Masked
  pixelDepthOffsetMode=PDOM_Legacy
  twoSided=true
  usedShadingModels=Default Lit
 }
 nodes [
  TextureSample() {texture=&/body_future_broken_window_RGB}
  TextureSample() {samplerType=SAMPLERTYPE_Normal, texture=&/body_future_broken_window_N}
 ]
}
